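# Niagara system that builds glowing magma hair for the MagmaBreak player body from mesh particles.
# One HairStrands emitter bursts eight persistent strands, using spawn groups to pick each one's scale, rotation, bone and offset from arrays.
# Strands follow the skeletal mesh every update, passing an interpolated position average and per-particle randomness into dynamic material parameters.
# Two mesh renderers draw them, with extra glow and dark parameters and time-of-day-scaled colors.
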NiagaraSystem /BRCosmetics/Characters/Player/Male/Medium/Bodies/M_MED_MagmaBreak/FX/NS_MagmaBreak_Hair
dependencies {
 usedDataInterfaces [
  &NiagaraDataInterfaceArrayFloat3
  &NiagaraDataInterfaceArrayInt32
  &NiagaraDataInterfaceSkeletalMesh
  &NiagaraDataInterfaceArrayColor
 ]
 usedDynamicInputs [&SelectLinearColorFromArray, &MakeIntFromFloat]
 usedModules [
  &self:SystemSpawnScript.SystemScriptSource.SystemScriptGraph.NiagaraNodeAssignment_0.SetVariables_#1
  &self:SystemSpawnScript.SystemScriptSource.SystemScriptGraph.NiagaraNodeAssignment_1.SetVariables_#2
  &SystemState
 ]
}
emitters {HairStrands {
  emitterUpdateScript [
   &EmitterState(
    Inactive Response=::NewEnumerator1
    Life Cycle Mode=::NewEnumerator1
    Loop Behavior=::NewEnumerator1
    Loop Duration Mode=::NewEnumerator1
    UseLoopDelay=false
   )
   &SpawnBurst_Instantaneous(Spawn Count=8, Use Spawn Probability=false, UseLoopCountLimit=false)
  ]
  particleSpawnScript [
   &FN_SetSpawnGroup(Number of Spawn Groups=8)
   &InitializeParticle(
    Color=<rgba 0.0469 0.0469 0.0469 1>
    Lifetime=&RandomRangeFloat(Minimum=0.45, Maximum=0.6)
    Mesh Scale=&SelectVectorFromArray(
     Direct Array Index=&SpawnGroup()
     Vector Selection Array=DataInterface() {propertyValues {FloatData [<0.8 0.8 0.9>, <1 1 0.7>, <1 1 1>, <0.9 0.8 0.8>, <0.8 0.8 0.8>, <1 1 1>, <1 1 1>]}}
    )
    Mesh Scale Mode=::NewEnumerator4
    UsePositionOffset=false
    Write Material Random=true
   )
   &InitialMeshOrientation(
    Mesh Axis To Orient=<0 0 1>
    Mesh Orientation Mode=::NewEnumerator2
    Orientation Vector=&SelectVectorFromArray(
     Direct Array Index=&SpawnGroup()
     Vector Selection Array=DataInterface() {propertyValues {FloatData [<1 0 0>, <1 0 0>, <1 0 0>, <1 0 0>, <1 0 0>, <1 0 0>, <1 0 0>]}}
    )
    Rotation=&SelectVectorFromArray(
     Direct Array Index=&SpawnGroup()
     Vector Selection Array=DataInterface() {propertyValues {FloatData [
        <0.0991 0.036 0.5>
        <0.045 0 0.5>
        <0.0721 -0.0631 0.5>
        <0.1351 -0.1261 0.5>
        <0.018 0 0.5>
        <0.1261 0 0.5>
        <0.1515 -0.036 0.5>
        <0.1892 0.018 0.5>
       ]}}
    )
    Use Rotation Vector=true
   )
   /SetVariables(Particles.GlowScale:Vector3f=&Multiply_Vector(A=$Particles.Scale, B=<1.3 1.3 1.3>)) {isSetParameters=true}
   &CameraOffset() {enabled=false}
   /SetVariables(Particles.RandFloat:NiagaraFloat=&RandomRangeFloat(Minimum=0, Maximum=1)) {isSetParameters=true}
  ]
  particleUpdateScript [
   &ParticleState(Kill Particles When Lifetime Has Elapsed=false)
   &SkeletalMeshLocation(
    Bone / Socket Index=&SelectIntFromArray(
     Direct Array Index=&SpawnGroup()
     Int Selection Array=DataInterface() {propertyValues {IntData [0,0,0,0,1,1,2,2]}}
    )
    Bone Sampling Mode=::NewEnumerator4
    MaskSpawnGroup=false
    Sampled Position Offset=&SelectVectorFromArray(
     Direct Array Index=&SpawnGroup()
     Vector Selection Array=DataInterface() {propertyValues {FloatData [<0 5 4.4>, <-1 2 4.4>, <0 -4 4.4>, <-1 -5.5 2.5>, <-2 -2 4.3776>, <0 3 4.4>, <-1 -3 4>, <-3 4 4>]}}
    )
    Skeletal Mesh=$System.SkeletalMesh
   ) {module=SkeletalMeshLocation001}
   &InterpolateOverTime(Rate Of Change=40, Target Vector Value:NiagaraPosition=$Particles.Position, Variable Type=::NewEnumerator2) {module=InterpolateOverTime001}
   &DynamicMaterialParameters(
    Index 0 Param 1=&UniformAOrBFloat(
     A=0.9
     B=1.2
     Distribution Weight=$Particles.RandFloat
     Evaluation Type=::NewEnumerator0
     Recalculate Random Each Loop=true
     Override Seed=false
    )
    Index 0 Param 2=&UniformAOrBFloat(
     A=0.9
     B=1.2
     Distribution Weight=$Particles.RandFloat
     Evaluation Type=::NewEnumerator0
     Recalculate Random Each Loop=true
     Override Seed=false
    )
    Index 0 Param 3=1
    Index 0 Param 4=0.81
    Index 1 Param 1=&MakeFloatFromVector(Vector=$StackContext.InterpolateOverTime001.InterpolateOverTime_V3_.Moving Average, Channel=::NewEnumerator0)
    Index 1 Param 2=&MakeFloatFromVector(Vector=$StackContext.InterpolateOverTime001.InterpolateOverTime_V3_.Moving Average, Channel=::NewEnumerator1)
    Index 1 Param 3=&MakeFloatFromVector(Vector=$StackContext.InterpolateOverTime001.InterpolateOverTime_V3_.Moving Average, Channel=::NewEnumerator2)
    Index 2 Param 1=$Particles.RandFloat
    Module.Index0Param0Write=true
    Module.Index0Param1Write=true
    Module.Index0Param2Write=true
    Module.Index0Param3Write=true
    Module.Index1Param0Write=true
    Module.Index1Param1Write=true
    Module.Index1Param2Write=true
    Module.Index1Param3Write=true
    Module.Index2Param0Write=true
    Module.Index2Param1Write=true
    Module.Index2Param2Write=true
    Module.Index2Param3Write=true
    Write Parameter Index 1=true
    Write Parameter Index 2=true
   ) {module=DynamicMaterialParameters001}
   /SetVariables(
    Particles.Dynamic Parameter Dark:Vector4f=&MakeVector4(
     X=&UniformAOrBFloat(
      A=1.1
      B=1.4
      Distribution Weight=$Particles.RandFloat
      Evaluation Type=::NewEnumerator0
      Recalculate Random Each Loop=true
      Override Seed=false
     )
     Y=&UniformAOrBFloat(
      A=1.1
      B=1.4
      Distribution Weight=$Particles.RandFloat
      Evaluation Type=::NewEnumerator0
      Recalculate Random Each Loop=true
      Override Seed=false
     )
     Z=1
     W=1.06
    )
    Particles.Dynamic Parameter Dark1:Vector4f=&MakeVector4(X=$Particles.RandFloat, Y=1, Z=1, W=1)
   ) {isSetParameters=true}
   /SetVariables(Particles.Inner Color:LinearColor=&MultiplyLinearColor(
     Linear Color=<rgba 7 0.923 0.1094 1>
     Scale Factor=&MakeVector4_FromFloat(Float=&FNTimeOfDayScalar(Night Scale=0.7, Morning Scale=1, Evening Scale=0.85))
    )) {isSetParameters=true}
   /SetVariables(Particles.Color=&MultiplyLinearColor(
     Linear Color=<rgba 1 0.0597 0.0199 0.568>
     Scale Factor=&MakeVector4_FromFloat(Float=&FNTimeOfDayScalar(Night Scale=0.7, Morning Scale=1, Evening Scale=0.85))
    )) {isSetParameters=true}
   /SetVariables(Particles.Dynamic Parameter Glow:Vector4f=&MakeVector4(X=7, Y=2, Z=1, W=1.06)) {isSetParameters=true}
   &ScaleMeshSize(Scale Factor=&SelectVectorFromArray(
     Direct Array Index=&SpawnGroup()
     Vector Selection Array=DataInterface() {propertyValues {FloatData [<1 1 1>, <1 1 1>, <1 1 1>, <1 1 1>, <1 1 1>, <1 1 1>, <1 1 1>, <1 1 1>]}}
    ))
  ]
  rendererClasses [&NiagaraMeshRendererProperties]
  renderers [
   {rendererClass=&NiagaraMeshRendererProperties, rendererIndex=0}
   {rendererClass=&NiagaraMeshRendererProperties, rendererIndex=1}
  ]
 }}
summary {
 systemName=NS_MagmaBreak_Hair
 emitters [{emitterName=HairStrands, rendererClasses [&NiagaraMeshRendererProperties]}]
 userVariables [
  User.bIsFrontEnd {type=NiagaraBool}
  User.bIsFrontEndPreview {type=NiagaraBool}
  User.bShowBackFX {type=NiagaraBool, defaultValue=true}
  User.VariantID {type=NiagaraFloat, defaultValue=0, description=""}
 ]
}
