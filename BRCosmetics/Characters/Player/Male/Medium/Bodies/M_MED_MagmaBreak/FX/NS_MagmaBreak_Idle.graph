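# Idle Niagara effect for the MagmaBreak character cosmetic, adding embers, glows and shoulder meshes attached to skeletal mesh bones.
# Embers continuously spawns curl-noise sprites, while Glow holds five flickering sprite points with lights, adjusted for platform and hidden head.
# Shoulders keeps two layered meshes on bones, passing smoothed position to dynamic material parameters so the material can react to movement.
# Time-of-day scales tint all emitters; front-end modifiers retune Glow and Shoulders.
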NiagaraSystem /BRCosmetics/Characters/Player/Male/Medium/Bodies/M_MED_MagmaBreak/FX/NS_MagmaBreak_Idle
dependencies {
 usedDataInterfaces [
  &NiagaraDataInterfaceArrayFloat3
  &NiagaraDataInterfaceArrayInt32
  &NiagaraDataInterfaceVectorCurve
  &NiagaraDataInterfaceCurve
  &NiagaraDataInterfaceArrayFloat2
  &NiagaraDataInterfaceArrayFloat
  &NiagaraDataInterfaceColorCurve
  &NiagaraDataInterfacePlatformSet
  &NiagaraDataInterfaceSkeletalMesh
  &NiagaraDataInterfaceArrayColor
 ]
 usedDynamicInputs [
  &Lerp_Float
  &SelectLinearColorFromArray
  &MakeIntFromFloat
 ]
 usedModules [
  &self:SystemSpawnScript.SystemScriptSource.SystemScriptGraph.NiagaraNodeAssignment_0.SetVariables_#1
  &self:SystemSpawnScript.SystemScriptSource.SystemScriptGraph.NiagaraNodeAssignment_1.SetVariables_#2
  &SystemState
 ]
}
emitters {
 Embers {
  emitterUpdateScript [
   &EmitterState()
   &SpawnRate(SpawnRate=11, Use Spawn Probability=false)
   &SpawnRate(SpawnGroup=1, SpawnRate=7, Use Spawn Probability=false) {enabled=false, module=SpawnRate001}
   &SpawnRate(SpawnGroup=2, SpawnRate=7, Use Spawn Probability=false) {enabled=false, module=SpawnRate002}
  ]
  particleSpawnScript [
   &InitializeParticle(
    Lifetime Max=0.7
    Lifetime Min=0.5
    Lifetime Mode=::NewEnumerator1
    Sprite Rotation Mode=::NewEnumerator2
    Sprite Size Mode=::NewEnumerator1
    Uniform Sprite Size Max=4
    Uniform Sprite Size Min=1.5
    UsePositionOffset=false
    Write Material Random=true
   )
   &AddVelocity(Velocity=&RandomRangeVector(
     Minimum=&SelectVectorFromArray(
      Direct Array Index=&SpawnGroup()
      Vector Selection Array=DataInterface() {propertyValues {FloatData [<-10 10 -5>, <-10 -10 5>, <-10 -10 5>]}}
     )
     Maximum=&SelectVectorFromArray(
      Direct Array Index=&SpawnGroup()
      Vector Selection Array=DataInterface() {propertyValues {FloatData [<10 20 20>, <10 -30 15>, <10 -30 15>]}}
     )
    ))
   &SkeletalMeshLocation(
    Bone / Socket Index=&SelectIntFromArray(Direct Array Index=&SpawnGroup(), Int Selection Array=DataInterface() {propertyValues {IntData [2,0,1]}})
    Bone Sampling Mode=::NewEnumerator4
    MaskSpawnGroup=false
    Position Sampling=::NewEnumerator1
    Sampled Position Offset=&Add_Vector(
     A=&SelectVectorFromArray(
      Direct Array Index=&SpawnGroup()
      Vector Selection Array=DataInterface() {propertyValues {FloatData [<-3 10 0>, <0 7 0>, <0 -7 0>, <0 0 0>]}}
     )
     B=&Multiply_Vector(
      A=&RandomRangeVector(Minimum=<-1 -1 -1>, Maximum=<1 1 1>)
      B=&SelectVectorFromArray(
       Direct Array Index=&SpawnGroup()
       Vector Selection Array=DataInterface() {propertyValues {FloatData [<3 3 3>, <1 1 1>]}}
      )
     )
    )
    Skeletal Mesh=$System.SkeletalMesh
    Source Movement Scale:Vector3f=<0.5 0.5 0.5>
   )
  ]
  particleUpdateScript [
   &CurlNoiseForce(
    Noise Frequency=9
    Noise Quality / Cost=::NewEnumerator0
    Noise Strength=&RandomRangeFloat(Minimum=15, Maximum=25)
    Pan Noise=true
    Pan Noise Field=&RandomRangeVector(Minimum=<0 0 0>, Maximum=<1 1 1>)
    Random Seed=2
   )
   &InheritSourceMovement(
    Applied Position Scale=&VectorFromCurve(VectorCurve=DataInterface() {propertyValues {
       XCurve {keys [<at 0 1 RCIM_Cubic>,<at 0.3 0>]}
       YCurve {keys [<at 0 1 RCIM_Cubic>,<at 0.3 0>]}
       ZCurve {keys [<at 0 1 RCIM_Cubic>,<at 0.3 0>]}
      }})
    ApplyForce=false
    ApplyPosition=true
    ApplyPositionFalloff=false
   ) {enabled=false}
   &SolveForcesAndVelocity(Clamp Velocity=false, Limit Acceleration=false)
   &ParticleState()
   &ScaleSpriteSizeBySpeed(Max Scale Factor=<1.2 4>, Min Scale Factor=<0.9 1.4>)
   &ScaleSpriteSize(Uniform Curve Sprite Scale=DataInterface() {propertyValues {Curve {keys [
        <at 0 0 RCIM_Cubic>
        {arriveTangent=1.5714, interpMode=RCIM_Cubic, leaveTangent=1.5714, time=0.4, value=1}
        {arriveTangent=-1.0833, interpMode=RCIM_Cubic, leaveTangent=-1.0833, time=0.7, value=1.1}
        <at 1 0.35 RCIM_Cubic>
       ]}}})
   /SetVariables(Particles.Color=&MultiplyLinearColor(
     Linear Color=$Particles.Initial.Color
     Scale Factor=&MakeVector4_FromLinearColor(LinearColor=&FNTimeOfDayColor(
       Night Color=<rgba 0.277 0.2868 0.4792 0.85>
       Morning Color=<rgba 0.8906 0.5659 0.5659 1>
       Day Color=<rgba 1 1 1 1>
       Evening Color=<rgba 1 1 1 1>
      ))
    )) {isSetParameters=true}
   &ScaleColor(
    Scale Alpha=&FloatFromCurve(FloatCurve=DataInterface() {propertyValues {Curve {keys [<at 0 1>, {interpMode=RCIM_Cubic, leaveTangent=-4.1367, tangentMode=RCTM_Break, time=0.65, value=0.95}, <at 1 0>]}}})
    ScaleA=true
    ScaleRGB=true
   )
   &DynamicMaterialParameters(
    Index 0 Param 1=&FloatFromCurve(FloatCurve=DataInterface() {propertyValues {Curve {keys [
         <at -0.1 0.05>
         {interpMode=RCIM_Cubic, leaveTangent=1.2819, tangentMode=RCTM_Break, time=0.4, value=0.2}
         {arriveTangent=0.0217, tangentMode=RCTM_User, time=0.7, value=0.95}
        ]}}})
    Index 0 Param 2=&RandomRangeFloat(Minimum=0, Maximum=1)
    Index 0 Param 3=&RandomRangeFloat(Minimum=0, Maximum=1)
    Module.Index0Param0Write=true
    Module.Index0Param1Write=true
    Module.Index0Param2Write=true
    Module.Index0Param3Write=true
   )
  ]
  rendererClasses [&NiagaraSpriteRendererProperties]
  renderers [{rendererClass=&NiagaraSpriteRendererProperties}]
 }
 Glow {
  emitterUpdateScript [
   &EmitterState(
    Inactive Response=::NewEnumerator1
    Life Cycle Mode=::NewEnumerator1
    Loop Behavior=::NewEnumerator1
    Loop Duration Mode=::NewEnumerator1
    UseLoopDelay=false
   )
   &SpawnBurst_Instantaneous(Spawn Count=5, Use Spawn Probability=false, UseLoopCountLimit=false)
  ]
  particleSpawnScript [
   &FN_SetSpawnGroup(Number of Spawn Groups=5)
   &InitializeParticle(
    Color=<rgba 1.1 0.3657 0.0044 1>
    Sprite Size=&SelectVector2DFromArray(
     Direct Array Index=&SpawnGroup()
     Vector 2D Selection Array=DataInterface() {propertyValues {FloatData [<40 80>, <40 80>, <40 40>, <50 90>]}}
    )
    Sprite Size Mode=::NewEnumerator4
    UsePositionOffset=false
    Write Material Random=true
   )
   &CameraOffset(Camera Offset Amount=&SelectFloatFromArray(Direct Array Index=&SpawnGroup(), Float Selection Array=DataInterface() {propertyValues {FloatData [5,5,5,2]}}))
   /SetVariables(Particles.LightsVisibilityTag:NiagaraInt32=&SelectIntFromArray(Direct Array Index=&SpawnGroup(), Int Selection Array=DataInterface() {propertyValues {IntData [1,1,1,1,0]}})) {isSetParameters=true}
  ]
  particleUpdateScript [
   &ParticleState(Kill Particles When Lifetime Has Elapsed=false, Let Infinitely Lived Particles Die When Emitter Deactivates=true)
   &SkeletalMeshLocation(
    Bone / Socket Index=&SelectIntFromArray(Direct Array Index=&SpawnGroup(), Int Selection Array=DataInterface() {propertyValues {IntData [0,1,2,3]}})
    Bone Sampling Mode=::NewEnumerator4
    MaskSpawnGroup=false
    Sampled Position Offset=&Add_Vector(
     A=&SelectVectorFromArray(
      Direct Array Index=&SpawnGroup()
      Vector Selection Array=DataInterface() {propertyValues {FloatData [<3 0 4>, <3 0 4>, <-3 11 0>, <-1 -13.3 0>, <13.7532 -36.9582 0>]}}
     )
     B=<0 0 0>
    )
    Skeletal Mesh=$System.SkeletalMesh
   )
   &SpriteFacingAndAlignment(
    Sprite Alignment=&RotateVectorByQuaternion(
     Vector To Rotate=&SelectVectorFromArray(
      Direct Array Index=&SpawnGroup()
      Vector Selection Array=DataInterface() {propertyValues {FloatData [<1 -1 0.4313>, <1 1 0.4313>, <1 0 0>, <1 0.7027 0>]}}
     )
     Quaternion=$Particles.SkeletalMeshLocation.SampledRotation
    )
    WriteAlignment=true
    WriteFacing=true
   )
   &ScaleColor(
    Scale Alpha=&Multiply_Float(
     A=&Multiply_Float(
      A=&FN_PickFloatDependOnPlatform(TargetPlatformActive=1, TargetPlatformNoActive=0.4)
      B=&FN_FrontEndFloatModifier(
       Gameplay=&MakeCustomFloatFromBool(
        Bool=$User.bisHeadHidden
        True Float=&MakeCustomFloatFromBool(
         Bool=&SetBoolByIntComparison(A=&SpawnGroup(), B=3, Comparison Type=::NewEnumerator6)
         True Float=0.22
         False Float=0
        )
        False Float=0.22
       )
       Front End=0.1
      )
     )
     B=&Waveform("Waveform [1]"=::NewEnumerator12, "[1] Frequency"=7, "[1] Amplitude Min/Max"=<0.75 1>)
    )
    Scale RGB=&VectorFromFloat(Value=&FNTimeOfDayScalar(Night Scale=0.66, Morning Scale=1, Evening Scale=0.85))
    ScaleA=true
    ScaleRGB=true
   )
   &FN_Light_Attributes(
    Color=<rgba 24 7.5112 0 1>
    Create Color Attribute=true
    Create Exponent Attribute=true
    Create Radius Attribute=true
    Exponent=2.3
    Radius=50
    Scale Color=&Multiply_Float(
     A=&FNTimeOfDayScalar(Night Scale=0.7, Morning Scale=0.9, Evening Scale=0.85)
     B=&Waveform("Waveform [1]"=::NewEnumerator12, "[1] Frequency"=7, "[1] Amplitude Min/Max"=<0.6 1>)
    )
   )
  ]
  rendererClasses [&NiagaraLightRendererProperties, &NiagaraSpriteRendererProperties]
  renderers [
   {rendererClass=&NiagaraLightRendererProperties, rendererIndex=0}
   {rendererClass=&NiagaraSpriteRendererProperties, rendererIndex=1}
  ]
 }
 Shoulders {
  emitterUpdateScript [
   &EmitterState(
    Inactive Response=::NewEnumerator1
    Life Cycle Mode=::NewEnumerator1
    Loop Behavior=::NewEnumerator1
    Loop Duration Mode=::NewEnumerator1
    UseLoopDelay=false
   )
   &SpawnBurst_Instantaneous(Spawn Count=2, Use Spawn Probability=false, UseLoopCountLimit=false)
  ]
  particleSpawnScript [
   &FN_SetSpawnGroup(Number of Spawn Groups=2)
   &InitializeParticle(
    Color=<rgba 0.0469 0.0469 0.0469 1>
    Lifetime=&RandomRangeFloat(Minimum=0.45, Maximum=0.6)
    Mesh Scale=&SelectVectorFromArray(
     Direct Array Index=&SpawnGroup()
     Vector Selection Array=DataInterface() {propertyValues {FloatData [<1.15 1.5 1.25>, <1.15 1.5 1.25>]}}
    )
    Mesh Scale Mode=::NewEnumerator4
    UsePositionOffset=false
    Write Material Random=true
   )
   &InitialMeshOrientation(
    Mesh Orientation Mode=::NewEnumerator2
    Orientation Vector=&SelectVectorFromArray(
     Direct Array Index=&SpawnGroup()
     Vector Selection Array=DataInterface() {propertyValues {FloatData [<0.5 -1 0>, <0.5 -1 0>]}}
    )
    Rotation=&FN_FrontEndVector3Modifier(
     Gameplay=&SelectVectorFromArray(
      Direct Array Index=&SpawnGroup()
      Vector Selection Array=DataInterface() {propertyValues {FloatData [<0 0.025 0.065>, <0 -0.025 -0.2143>]}}
     )
     Front End=&SelectVectorFromArray(
      Direct Array Index=&SpawnGroup()
      Vector Selection Array=DataInterface() {propertyValues {FloatData [<0 0.025 0.065>, <0 -0.025 -0.21>]}}
     )
    )
    Use Rotation Vector=true
   )
   /SetVariables(Particles.GlowScale:Vector3f=&Multiply_Vector(A=$Particles.Scale, B=<1.3 1.3 1.3>)) {isSetParameters=true}
   /SetVariables(Particles.SplineVisibility:NiagaraInt32=&SelectIntFromArray(Direct Array Index=&SpawnGroup(), Int Selection Array=DataInterface() {propertyValues {IntData [0,1]}})) {isSetParameters=true}
  ]
  particleUpdateScript [
   &ParticleState(Kill Particles When Lifetime Has Elapsed=false)
   &SkeletalMeshLocation(
    Bone / Socket Index=&SpawnGroup()
    Bone Sampling Mode=::NewEnumerator4
    MaskSpawnGroup=false
    Sampled Position Offset=&FN_FrontEndVector3Modifier(
     Gameplay=&SelectVectorFromArray(
      Direct Array Index=&SpawnGroup()
      Vector Selection Array=DataInterface() {propertyValues {FloatData [<4.6 2.3 3>, <4.3 -2.3 3>]}}
     )
     Front End=&SelectVectorFromArray(
      Direct Array Index=&SpawnGroup()
      Vector Selection Array=DataInterface() {propertyValues {FloatData [<4.6 2.1 3>, <4 -1.4267 1.8141>]}}
     )
    )
    Skeletal Mesh=$System.SkeletalMesh
   ) {module=SkeletalMeshLocation001}
   &InterpolateOverTime(Rate Of Change=12, Target Vector Value:NiagaraPosition=$Particles.Position, Variable Type=::NewEnumerator2) {module=InterpolateOverTime001}
   &DynamicMaterialParameters(
    Index 0 Param 1=&FN_FrontEndFloatModifier(Gameplay=2, Front End=1.8)
    Index 0 Param 2=&FN_FrontEndFloatModifier(Gameplay=2, Front End=1.6)
    Index 0 Param 3=1
    Index 0 Param 4=0.81
    Index 1 Param 1=&MakeFloatFromVector(Vector=$StackContext.InterpolateOverTime001.InterpolateOverTime_V3_.Moving Average, Channel=::NewEnumerator0)
    Index 1 Param 2=&MakeFloatFromVector(Vector=$StackContext.InterpolateOverTime001.InterpolateOverTime_V3_.Moving Average, Channel=::NewEnumerator1)
    Index 1 Param 3=&MakeFloatFromVector(Vector=$StackContext.InterpolateOverTime001.InterpolateOverTime_V3_.Moving Average, Channel=::NewEnumerator2)
    Index 2 Param 1=&RandomRangeFloat(Minimum=0, Maximum=1)
    Module.Index0Param0Write=true
    Module.Index0Param1Write=true
    Module.Index0Param2Write=true
    Module.Index0Param3Write=true
    Module.Index1Param0Write=true
    Module.Index1Param1Write=true
    Module.Index1Param2Write=true
    Module.Index1Param3Write=true
    Module.Index2Param0Write=true
    Module.Index2Param1Write=true
    Module.Index2Param2Write=true
    Module.Index2Param3Write=true
    Write Parameter Index 1=true
    Write Parameter Index 2=true
   ) {module=DynamicMaterialParameters001}
   /SetVariables(Particles.Dynamic Parameter Dark:Vector4f=&MakeVector4(X=&FN_FrontEndFloatModifier(Gameplay=3.6, Front End=3.2), Y=2, Z=1, W=1.06)) {isSetParameters=true}
   /SetVariables(Particles.Inner Color:LinearColor=&MultiplyLinearColor(
     Linear Color=<rgba 2.8 0.5124 0.0292 0.822>
     Scale Factor=&MakeVector4_FromFloat(Float=&FNTimeOfDayScalar(Night Scale=0.9, Morning Scale=0.95, Evening Scale=0.9))
    )) {isSetParameters=true}
   /SetVariables(Particles.Color=&MultiplyLinearColor(
     Linear Color=<rgba 1 0.07 0.0208 0.853>
     Scale Factor=&MakeVector4_FromFloat(Float=&FNTimeOfDayScalar(Night Scale=0.66, Morning Scale=1, Evening Scale=0.85))
    )) {isSetParameters=true}
   /SetVariables(Particles.GlowColor=&MultiplyLinearColor(
     Linear Color=<rgba 3 0.018 0 0.518>
     Scale Factor=&MakeVector4_FromFloat(Float=&Multiply_Float(
       A=&FN_FrontEndFloatModifier(Gameplay=1, Front End=0.5)
       B=&FNTimeOfDayScalar(Night Scale=0.66, Morning Scale=1, Evening Scale=0.85)
      ))
    )) {isSetParameters=true}
   /SetVariables(Particles.Dynamic Parameter Glow:Vector4f=&MakeVector4(X=5, Y=2, Z=1, W=1.06)) {isSetParameters=true}
  ]
  rendererClasses [&NiagaraMeshRendererProperties]
  renderers [
   {rendererClass=&NiagaraMeshRendererProperties, rendererIndex=0}
   {rendererClass=&NiagaraMeshRendererProperties, rendererIndex=1}
   {rendererClass=&NiagaraMeshRendererProperties, rendererIndex=2}
   {rendererClass=&NiagaraMeshRendererProperties, rendererIndex=3}
   {rendererClass=&NiagaraMeshRendererProperties, rendererIndex=4}
   {rendererClass=&NiagaraMeshRendererProperties, rendererIndex=5}
  ]
 }
}
summary {
 systemName=NS_MagmaBreak_Idle
 emitters [
  {emitterName=Embers, rendererClasses [&NiagaraSpriteRendererProperties]}
  {emitterName=Shoulders, rendererClasses [&NiagaraMeshRendererProperties]}
  {emitterName=Glow, rendererClasses [&NiagaraLightRendererProperties, &NiagaraSpriteRendererProperties]}
 ]
 userVariables [
  User.bIsFrontEnd {type=NiagaraBool}
  User.bIsFrontEndPreview {type=NiagaraBool}
  User.bisHeadHidden {type=NiagaraBool}
  User.VariantID {type=NiagaraFloat, defaultValue=0, description=""}
 ]
}
