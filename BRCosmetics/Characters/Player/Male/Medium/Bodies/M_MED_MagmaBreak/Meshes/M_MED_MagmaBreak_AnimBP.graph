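# Animation Blueprint for the MagmaBreak medium male player body, which follows its attached parent mesh and applies body-specific corrections.
# The copied parent pose runs through the MagmaBreak Control Rig, which receives a targeting alpha, then a constant additive offset nudges the backpack attach bone.
# Interpolated additive scaling enlarges both upper-arm twist bones while the parachute is open.
# Each update converts the targeting and parachute booleans into those float alphas.
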AnimBlueprint /BRCosmetics/Characters/Player/Male/Medium/Bodies/M_MED_MagmaBreak/Meshes/M_MED_MagmaBreak_AnimBP
target_skeleton=&/M_MED_MagmaBreak_Skeleton
graphs [
 AnimGraph {nodes [
   n1 = Control Rig!(Source=Copy Pose From Mesh() {bUseAttachedParent=true}, Alpha, Get Is targeting:double=Get IsTargetingAlpha() -> (IsTargetingAlpha:double)) -> (Pose) {
    alphaBoolBlend {blendOption=Linear}
    controlRigAssetReference {blueprintRigClass=&/M_MED_MagmaBreak_ControlRig.M_MED_MagmaBreak_ControlRig_C}
   }
   n2 = "Transform (Modify) Bone - Bone: attach_backpack"(ComponentPose=Local To Component(LocalPose=@n1), Alpha=1, Translation="0,-1,0", Rotation="0,0,0", Scale="1,1,1") {
    translationMode=BMM_Additive
    translationSpace=BCS_BoneSpace
    alphaBoolBlend {blendOption=Linear}
    boneToModify {boneName=attach_backpack}
    translation {y=-1}
   }
   n3 = Get IsParachutingAlpha() -> (IsParachutingAlpha:double)
   n4 = "Transform (Modify) Bone - Bone: upperarm_twist_02_r"(ComponentPose=@n2, Alpha=@n3, Translation="0,0,0", Rotation="0,0,0", Scale="1.3,1.3,1.3") {
    scaleMode=BMM_Additive
    scaleSpace=BCS_BoneSpace
    alphaBoolBlend {blendOption=Linear}
    alphaScaleBiasClamp {bInterpResult=true}
    boneToModify {boneName=upperarm_twist_02_r}
    scale=<1.3 1.3 1.3>
   }
   n5 = "Transform (Modify) Bone - Bone: upperarm_twist_02_l"(ComponentPose=@n4, Alpha=@n3, Translation="0,0,0", Rotation="0,0,0", Scale="1.3,1.3,1.3") {
    scaleMode=BMM_Additive
    scaleSpace=BCS_BoneSpace
    alphaBoolBlend {blendOption=Linear}
    alphaScaleBiasClamp {bInterpResult=true}
    boneToModify {boneName=upperarm_twist_02_l}
    scale=<1.3 1.3 1.3>
   }
   Output Pose(Result=Component To Local(ComponentPose=@n5)) {layerGroup=None}
  ]}
 EventGraph {nodes [
   Event BlueprintUpdateAnimation()
   > Set IsTargetingAlpha(IsTargetingAlpha:double="To Float (Boolean)"(InBool=Get bIsTargeting())) -> (Output_Get:double)
   > Set IsParachutingAlpha(IsParachutingAlpha:double="To Float (Boolean)"(InBool=Get bIsParachuteOpened())) -> (Output_Get:double)
  ]}
]
variables [
 IsParachutingAlpha {type=number}
 IsTargetingAlpha {type=number}
]
