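# Physics asset for the MagmaBreak medium male character's face accessory/backpack mesh part, defining a ten-bone dyn_hair chain with surrounding upper-body shapes.
# Capsule bodies cover the upper spine, clavicles, neck, head and hair base, linked by angularly free constraints.
# The hair chain's root is locked to base_hair_m, and later links have small swing and twist limits, with one swing axis locked at dyn_hair_m_07.
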PhysicsAsset /BRCosmetics/Characters/Player/Male/Medium/Bodies/M_MED_MagmaBreak/Meshes/Parts/M_MED_MagmaBreak_FaceAcc_BackPack_Physics
bodies {
 base_hair_m [{shapeType=Capsule, length=0.505, radius=0.505, shapeName=base_hair_m_capsule, rotation=<rot 0 -90 0>}]
 clavicle_l [{shapeType=Capsule, length=7.5, radius=11, shapeName=clavicle_r_capsule, center=<12.4466 0.6577 -0.5013>, rotation=<rot 90 0 0>}]
 clavicle_r [{shapeType=Capsule, length=7.5, radius=10.5, shapeName=clavicle_r_capsule, center=<-12.4466 -0.1598 0.4278>, rotation=<rot 90 0 0>}]
 dyn_hair_m_01 [{shapeType=Capsule, length=5, radius=5, shapeName=dyn_hair_m_01_capsule, center=<2.9657 0 0>, rotation=<rot 0 0 90>}]
 dyn_hair_m_02 [{shapeType=Capsule, length=8.8971, radius=4.6183, shapeName=dyn_hair_m_02_capsule, center=<2.847 0 0>, rotation=<rot 0 0 90>}]
 dyn_hair_m_03 [{shapeType=Capsule, length=10.1023, radius=4.6205, shapeName=dyn_hair_m_03_capsule, center=<2.5247 0 0>, rotation=<rot 0 0 90>}]
 dyn_hair_m_04 [{shapeType=Capsule, length=10.1741, radius=5.2706, shapeName=dyn_hair_m_04_capsule, center=<2.2338 0 0>, rotation=<rot 0 0 90>}]
 dyn_hair_m_05 [{shapeType=Capsule, length=10.1676, radius=4.7944, shapeName=dyn_hair_m_05_capsule, center=<2.14 0 0>, rotation=<rot 0 0 90>}]
 dyn_hair_m_06 [{shapeType=Capsule, length=9.6102, radius=3.5605, shapeName=dyn_hair_m_06_capsule, center=<2.6458 0 0>, rotation=<rot 0 0 90>}]
 dyn_hair_m_07 [{shapeType=Capsule, length=7.9195, radius=3, shapeName=dyn_hair_m_07_capsule, center=<2.5662 0 0>, rotation=<rot 0 0 90>}]
 dyn_hair_m_08 [{shapeType=Capsule, length=6.1769, radius=2, shapeName=dyn_hair_m_08_capsule, center=<1.4068 0 0>, rotation=<rot 0 0 90>}]
 dyn_hair_m_09 [{shapeType=Capsule, length=1.5, radius=1.4244, shapeName=dyn_hair_m_09_capsule, center=<1.5519 0 0>, rotation=<rot 90 0 0>}]
 dyn_hair_m_10 [{shapeType=Capsule, length=2.6952, radius=1.2016, shapeName=dyn_hair_m_10_capsule, center=<2.3485 0 0>, rotation=<rot 90 0 0>}]
 head [{shapeType=Capsule, length=6.9589, radius=6.4589, shapeName=head_capsule, center=<0 1.3623 0>, rotation=<rot 90 0 0>}]
 neck_01 [{shapeType=Capsule, length=1.5, radius=6.5, shapeName=neck_01_capsule, center=<1.6916 0 0>, rotation=<rot 90 0 0>}]
 neck_02 [{shapeType=Capsule, length=2, radius=6, shapeName=neck_02_capsule, center=<0 -0.7661 0>, rotation=<rot 90 0 0>}]
 spine_05 [{shapeType=Capsule, length=16, radius=14, shapeName=spine_05_capsule, center=<0.5985 0.1258 0>, rotation=<rot 0 -90 0>}]
}
constraints [
 {
  bone1Name=base_hair_m
  bone2Name=head
  swing1LimitDegrees=45
  swing1Motion=Free
  swing2LimitDegrees=45
  swing2Motion=Free
  twistLimitDegrees=45
  twistMotion=Free
 }
 {
  bone1Name=dyn_hair_m_01
  bone2Name=base_hair_m
  swing1LimitDegrees=5
  swing1Motion=Locked
  swing2LimitDegrees=10
  swing2Motion=Locked
  twistLimitDegrees=5
  twistMotion=Locked
 }
 {bone1Name=dyn_hair_m_02, bone2Name=dyn_hair_m_01, swing1LimitDegrees=5, swing2LimitDegrees=5, twistLimitDegrees=5}
 {bone1Name=dyn_hair_m_03, bone2Name=dyn_hair_m_02, swing1LimitDegrees=10, swing2LimitDegrees=10, twistLimitDegrees=5}
 {bone1Name=dyn_hair_m_04, bone2Name=dyn_hair_m_03, swing1LimitDegrees=5, swing2LimitDegrees=15, twistLimitDegrees=5}
 {bone1Name=dyn_hair_m_05, bone2Name=dyn_hair_m_04, swing1LimitDegrees=10, swing2LimitDegrees=30, twistLimitDegrees=5}
 {bone1Name=dyn_hair_m_06, bone2Name=dyn_hair_m_05, swing1LimitDegrees=10, swing2LimitDegrees=15, twistLimitDegrees=5}
 {bone1Name=dyn_hair_m_07, bone2Name=dyn_hair_m_06, swing1LimitDegrees=5, swing2LimitDegrees=5, swing2Motion=Locked, twistLimitDegrees=5}
 {bone1Name=dyn_hair_m_08, bone2Name=dyn_hair_m_07, swing1LimitDegrees=10, swing2LimitDegrees=20, twistLimitDegrees=5}
 {bone1Name=dyn_hair_m_09, bone2Name=dyn_hair_m_08, swing1LimitDegrees=10, swing2LimitDegrees=10, twistLimitDegrees=5}
 {
  bone1Name=head
  bone2Name=neck_02
  swing1LimitDegrees=45
  swing1Motion=Free
  swing2LimitDegrees=45
  swing2Motion=Free
  twistLimitDegrees=45
  twistMotion=Free
 }
 {
  bone1Name=neck_01
  bone2Name=spine_05
  swing1LimitDegrees=45
  swing1Motion=Free
  swing2LimitDegrees=45
  swing2Motion=Free
  twistLimitDegrees=45
  twistMotion=Free
 }
 {
  bone1Name=neck_02
  bone2Name=neck_01
  swing1LimitDegrees=45
  swing1Motion=Free
  swing2LimitDegrees=45
  swing2Motion=Free
  twistLimitDegrees=45
  twistMotion=Free
 }
 {
  bone1Name=clavicle_l
  bone2Name=spine_05
  swing1LimitDegrees=45
  swing1Motion=Free
  swing2LimitDegrees=45
  swing2Motion=Free
  twistLimitDegrees=45
  twistMotion=Free
 }
 {
  bone1Name=clavicle_r
  bone2Name=spine_05
  swing1LimitDegrees=45
  swing1Motion=Free
  swing2LimitDegrees=45
  swing2Motion=Free
  twistLimitDegrees=45
  twistMotion=Free
 }
 {bone1Name=dyn_hair_m_10, bone2Name=dyn_hair_m_09, swing1LimitDegrees=10, swing2LimitDegrees=10, twistLimitDegrees=5}
]
