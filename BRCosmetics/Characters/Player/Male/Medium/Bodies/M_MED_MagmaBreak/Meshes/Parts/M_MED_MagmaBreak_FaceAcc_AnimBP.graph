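# Post-process style AnimBlueprint for the MagmaBreak FaceAcc part, adding secondary motion to its dyn_hair bones atop the pose copied from the parent mesh.
# Output selects cached cape (also long backbling), backbling or normal dynamics, each using head-relative RigidBody simulation with overridden gravity.
# The backbling branch first applies head/spine constraints and additive bone corrections, weighted partly by head-pitch and crouch-walk values computed each update.
# Several other corrective chains never reach the output.
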AnimBlueprint /BRCosmetics/Characters/Player/Male/Medium/Bodies/M_MED_MagmaBreak/Meshes/Parts/M_MED_MagmaBreak_FaceAcc_AnimBP
target_skeleton=&/M_MED_MagmaBreak_FaceAcc_Skeleton
graphs [
 AnimGraph {nodes [
   n1 = RigidBody!(
    ComponentPose=Local To Component(LocalPose=Use cached pose 'StartingPose'())
    Alpha
    OverrideWorldGravity:Vector=Get GravityOverride()
    ExternalForce
   ) -> (Pose) {
    bClampLinearTranslationLimitToRefPose=true
    bOverrideWorldGravity=true
    overridePhysicsAsset=&/M_MED_MagmaBreak_FaceAcc_Physics
    simulationSpace=BaseBoneSpace
    alphaBoolBlend {blendOption=Linear}
    baseBoneRef {boneName=head}
    componentLinearAccScale=<0.24 0.24 0.24>
    componentLinearVelScale=<0.3 0.3 0.3>
    overrideWorldGravity {z=-980}
    simSpaceSettings {maxAngularAcceleration=1000, maxAngularVelocity=3, maxLinearAcceleration=250, maxLinearVelocity=500, velocityScaleZ=0, worldAlpha=0.25}
   }
   n2 = "Transform (Modify) Bone - Bone: dyn_hair_m_02"(ComponentPose=@n1, Alpha=1, Translation="0,0,0", Rotation="10,0,0", Scale="1,1,1") {
    rotationMode=BMM_Additive
    rotationSpace=BCS_BoneSpace
    translationMode=BMM_Additive
    translationSpace=BCS_BoneSpace
    alphaBoolBlend {blendOption=Linear}
    boneToModify {boneName=dyn_hair_m_02}
    rotation {pitch=10}
   }
   n3 = "Transform (Modify) Bone - Bone: dyn_hair_m_03"(ComponentPose=@n2, Alpha=1, Translation="0,0,0", Rotation="10,0,0", Scale="1,1,1") {
    rotationMode=BMM_Additive
    rotationSpace=BCS_BoneSpace
    translationMode=BMM_Additive
    translationSpace=BCS_BoneSpace
    alphaBoolBlend {blendOption=Linear}
    boneToModify {boneName=dyn_hair_m_03}
    rotation {pitch=10}
   }
   n4 = "Transform (Modify) Bone - Bone: dyn_hair_m_04"(ComponentPose=@n3, Alpha=1, Translation="0,0,2", Rotation="0,0,0", Scale="1,1,1") {
    rotationMode=BMM_Additive
    rotationSpace=BCS_BoneSpace
    translationMode=BMM_Additive
    translationSpace=BCS_BoneSpace
    alphaBoolBlend {blendOption=Linear}
    boneToModify {boneName=dyn_hair_m_04}
    translation {z=2}
   }
   n5 = "Transform (Modify) Bone - Bone: dyn_hair_m_05"(ComponentPose=@n4, Alpha=1, Translation="0,0,0", Rotation="0,0,0", Scale="1,1,1") {
    rotationMode=BMM_Additive
    rotationSpace=BCS_BoneSpace
    translationMode=BMM_Additive
    translationSpace=BCS_BoneSpace
    alphaBoolBlend {blendOption=Linear}
    boneToModify {boneName=dyn_hair_m_05}
   }
   n6 = "Transform (Modify) Bone - Bone: dyn_hair_m_06"(ComponentPose=@n5, Alpha=1, Translation="0,0,0", Rotation="0,0,0", Scale="1,1,1") {
    rotationMode=BMM_Additive
    rotationSpace=BCS_BoneSpace
    translationMode=BMM_Additive
    translationSpace=BCS_BoneSpace
    alphaBoolBlend {blendOption=Linear}
    alphaScaleBiasClamp {bInterpResult=true}
    boneToModify {boneName=dyn_hair_m_06}
   }
   Blend Poses by bool(BlendPose_0=Component To Local(ComponentPose=@n6), bActiveValue=Get bIsCrouching()) {blendType=Linear}
   Component To Local()
   Local To Component(LocalPose=Blend Poses by bool(
     BlendPose_0=Component To Local()
     BlendPose_1=Use cached pose 'BPPose01'()
     BlendTime_0=0.3
     BlendTime_1=0.3
     bActiveValue=Get IsCrouchWalking() -> (IsCrouchWalking:bool)
    ) {blendType=Linear})
   n7 = Get HeadUpFloat() -> (HeadUpFloat:double)
   n8 = "Transform (Modify) Bone - Bone: dyn_hair_m_02"!(ComponentPose=Local To Component(), Alpha=@n7, Rotation="5,0,0") -> (Pose) {
    rotationMode=BMM_Additive
    rotationSpace=BCS_BoneSpace
    alphaBoolBlend {blendOption=Linear}
    boneToModify {boneName=dyn_hair_m_02}
    rotation {pitch=5}
   }
   n9 = "Transform (Modify) Bone - Bone: dyn_hair_m_04"!(ComponentPose=@n8, Alpha=@n7, Rotation="25,0,0") -> (Pose) {
    rotationMode=BMM_Additive
    rotationSpace=BCS_BoneSpace
    alphaBoolBlend {blendOption=Linear}
    boneToModify {boneName=dyn_hair_m_04}
    rotation {pitch=25}
   }
   n10 = "Transform (Modify) Bone - Bone: dyn_hair_m_05"!(ComponentPose=@n9, Alpha=@n7, Rotation="25,0,0") -> (Pose) {
    rotationMode=BMM_Additive
    rotationSpace=BCS_BoneSpace
    alphaBoolBlend {blendOption=Linear}
    boneToModify {boneName=dyn_hair_m_05}
    rotation {pitch=25}
   }
   n11 = "Transform (Modify) Bone - Bone: dyn_hair_m_06"!(ComponentPose=@n10, Alpha=@n7, Rotation="25,0,0") -> (Pose) {
    rotationMode=BMM_Additive
    rotationSpace=BCS_BoneSpace
    alphaBoolBlend {blendOption=Linear}
    boneToModify {boneName=dyn_hair_m_06}
    rotation {pitch=25}
   }
   n12 = Get HeadUpFloat()
   n13 = "Transform (Modify) Bone - Bone: dyn_hair_m_02"(ComponentPose=Local To Component(LocalPose=Use cached pose 'StartingPose'()), Alpha=@n12, Rotation="20.000,0.000,0.000") {
    rotationMode=BMM_Additive
    rotationSpace=BCS_BoneSpace
    alphaBoolBlend {blendOption=Linear}
    boneToModify {boneName=dyn_hair_m_02}
    rotation {pitch=20}
   }
   n14 = "Transform (Modify) Bone - Bone: dyn_hair_m_04"(ComponentPose=@n13, Alpha=@n12, Rotation="25,0,0") {
    rotationMode=BMM_Additive
    rotationSpace=BCS_BoneSpace
    alphaBoolBlend {blendOption=Linear}
    boneToModify {boneName=dyn_hair_m_04}
    rotation {pitch=25}
   }
   n15 = "Transform (Modify) Bone - Bone: dyn_hair_m_05"(ComponentPose=@n14, Alpha=@n12, Rotation="10,0,0") {
    rotationMode=BMM_Additive
    rotationSpace=BCS_BoneSpace
    alphaBoolBlend {blendOption=Linear}
    boneToModify {boneName=dyn_hair_m_05}
    rotation {pitch=10}
   }
   n16 = "Transform (Modify) Bone - Bone: dyn_hair_m_06"(ComponentPose=@n15, Alpha=@n12, Rotation="10,0,0") {
    rotationMode=BMM_Additive
    rotationSpace=BCS_BoneSpace
    alphaBoolBlend {blendOption=Linear}
    boneToModify {boneName=dyn_hair_m_06}
    rotation {pitch=10}
   }
   n17 = Get HeadUpCrouchTargeting() -> (HeadUpCrouchTargeting:double)
   n18 = "Transform (Modify) Bone - Bone: dyn_hair_m_02"(ComponentPose=@n16, Alpha=@n17, Rotation="5,0,0") {
    rotationMode=BMM_Additive
    rotationSpace=BCS_BoneSpace
    alphaBoolBlend {blendOption=Linear}
    boneToModify {boneName=dyn_hair_m_02}
    rotation {pitch=5}
   }
   n19 = "Transform (Modify) Bone - Bone: dyn_hair_m_03"!(ComponentPose=@n18, Alpha=@n17, Rotation="10,0,0") -> (Pose) {
    rotationMode=BMM_Additive
    rotationSpace=BCS_BoneSpace
    alphaBoolBlend {blendOption=Linear}
    boneToModify {boneName=dyn_hair_m_03}
    rotation {pitch=10}
   }
   n20 = "Transform (Modify) Bone - Bone: dyn_hair_m_04"(ComponentPose=@n19, Alpha=@n17, Rotation="0,0,0") {rotationMode=BMM_Additive, rotationSpace=BCS_BoneSpace, alphaBoolBlend {blendOption=Linear}, boneToModify {boneName=dyn_hair_m_04}}
   n21 = "Transform (Modify) Bone - Bone: dyn_hair_m_07"!(ComponentPose=@n20, Alpha=@n17, Rotation="0,0,0") -> (Pose) {rotationMode=BMM_Additive, rotationSpace=BCS_BoneSpace, alphaBoolBlend {blendOption=Linear}, boneToModify {boneName=dyn_hair_m_07}}
   Local To Component(LocalPose=Blend Poses by bool(BlendPose_0=Component To Local(ComponentPose=@n21), BlendPose_1=Component To Local(ComponentPose=@n11), bActiveValue=Get bIsCrouching()) {blendType=Linear})
   Local To Component(LocalPose=Use cached pose 'BPPose01'())
   n22 = Get HeadUpFloat()
   n23 = "Transform (Modify) Bone - Bone: dyn_hair_m_02"!(ComponentPose, Alpha=@n22, Translation:Vector="0,0,0", Rotation="-12,0.01,-0.003", Scale:Vector="1,1,1") -> (Pose) {
    rotationMode=BMM_Additive
    rotationSpace=BCS_BoneSpace
    translationMode=BMM_Additive
    translationSpace=BCS_BoneSpace
    alphaBoolBlend {blendOption=Linear}
    alphaScaleBiasClamp {inRange {max=0, min=-25}, outRange {max=0, min=1}}
    boneToModify {boneName=dyn_hair_m_02}
    rotation=<rot -12 0.01 -0.003>
   }
   n24 = "Transform (Modify) Bone - Bone: dyn_hair_m_08"(ComponentPose=@n23, Alpha=@n22, Translation="2,0,0", Rotation="22,-0.012,-0.022", Scale="1,1,1") {
    rotationMode=BMM_Additive
    rotationSpace=BCS_BoneSpace
    translationMode=BMM_Additive
    translationSpace=BCS_BoneSpace
    alphaBoolBlend {blendOption=Linear}
    alphaScaleBiasClamp {inRange {max=0, min=-25}, outRange {max=0, min=1}}
    boneToModify {boneName=dyn_hair_m_08}
    rotation=<rot 22 -0.012 -0.022>
    translation {x=2}
   }
   "Transform (Modify) Bone - Bone: dyn_hair_m_09"(ComponentPose=@n24, Alpha=@n22, Translation="0,0,0", Rotation="-30,0,0", Scale="1,1,1") {
    rotationMode=BMM_Additive
    rotationSpace=BCS_BoneSpace
    translationMode=BMM_Additive
    translationSpace=BCS_BoneSpace
    alphaBoolBlend {blendOption=Linear}
    alphaScaleBiasClamp {inRange {max=0, min=-25}, outRange {max=0, min=1}}
    boneToModify {boneName=dyn_hair_m_09}
    rotation {pitch=-30}
   }
   n25 = "Transform (Modify) Bone - Bone: dyn_hair_m_04"(Alpha=1, Rotation="25,0,0") {
    rotationMode=BMM_Additive
    rotationSpace=BCS_BoneSpace
    alphaBoolBlend {blendOption=Linear}
    boneToModify {boneName=dyn_hair_m_04}
    rotation {pitch=25}
   }
   n26 = "Transform (Modify) Bone - Bone: dyn_hair_m_05"(ComponentPose=@n25, Alpha=1, Rotation="25,0,0") {
    rotationMode=BMM_Additive
    rotationSpace=BCS_BoneSpace
    alphaBoolBlend {blendOption=Linear}
    boneToModify {boneName=dyn_hair_m_05}
    rotation {pitch=25}
   }
   "Transform (Modify) Bone - Bone: dyn_hair_m_06"(ComponentPose=@n26, Alpha=1, Rotation="25,0,0") {
    rotationMode=BMM_Additive
    rotationSpace=BCS_BoneSpace
    alphaBoolBlend {blendOption=Linear}
    boneToModify {boneName=dyn_hair_m_06}
    rotation {pitch=25}
   }
   n27 = "Transform (Modify) Bone - Bone: dyn_hair_m_03"(Alpha=Get HeadUpFloat(), Rotation="10,0,0") {
    rotationMode=BMM_Additive
    rotationSpace=BCS_BoneSpace
    alphaBoolBlend {blendOption=Linear}
    boneToModify {boneName=dyn_hair_m_03}
    rotation {pitch=10}
   }
   n28 = "Transform (Modify) Bone - Bone: dyn_hair_m_02"!(ComponentPose=@n27, Alpha=Get Head Down() -> (Head Down:double), Rotation="15,0,0") -> (Pose) {
    rotationMode=BMM_Additive
    rotationSpace=BCS_BoneSpace
    alphaBoolBlend {blendOption=Linear}
    boneToModify {boneName=dyn_hair_m_02}
    rotation {pitch=15}
   }
   n29 = Get bIsTargeting()
   "Transform (Modify) Bone - Bone: dyn_hair_m_05"!(ComponentPose=@n28, bAlphaBoolEnabled:bool=@n29, Translation:Vector="0,0,2", Rotation="0,0,0") -> (Pose) {
    alphaInputType=Bool
    rotationMode=BMM_Additive
    rotationSpace=BCS_BoneSpace
    translationMode=BMM_Additive
    translationSpace=BCS_BoneSpace
    alphaBoolBlend {blendInTime=0.3, blendOption=Linear, blendOutTime=0.3}
    boneToModify {boneName=dyn_hair_m_05}
    translation {z=2}
   }
   n30 = Use cached pose 'Backpack Dynamics'()
   n31 = Use cached pose 'NormalDynamics'()
   n32 = Get is Wearing Backbling() -> (is Wearing Backbling:bool)
   Output Pose(Result=Blend Poses by bool(
     BlendPose_0=Use cached pose 'Cape Dynamics'()
     BlendPose_1=Blend Poses by bool(BlendPose_0=@n30, BlendPose_1=@n31, bActiveValue=@n32) {blendType=Linear}
     bActiveValue=Get is Wearing Cape() -> (is Wearing Cape:bool)
    ) {blendType=Linear}) {layerGroup=None}
   n33 = RigidBody(ComponentPose=Local To Component(LocalPose=Use cached pose 'StartingPose'()), OverrideWorldGravity=Get GravityOverride()) {
    bClampLinearTranslationLimitToRefPose=true
    bOverrideWorldGravity=true
    overridePhysicsAsset=&M_MED_MagmaBreak_FaceAcc_Physics
    simulationSpace=BaseBoneSpace
    alphaBoolBlend {blendOption=Linear}
    baseBoneRef {boneName=head}
    componentAppliedLinearAccClamp=<3000 3000 3000>
    componentLinearAccScale=<0.45 0.45 0.45>
    componentLinearVelScale=<0.45 0.4 0.4>
    overrideWorldGravity {z=-980}
    simSpaceSettings {maxAngularAcceleration=1000, maxAngularVelocity=3, maxLinearAcceleration=250, maxLinearVelocity=500, velocityScaleZ=0, worldAlpha=0.05}
   }
   Save cached pose 'NormalDynamics'(Pose=Component To Local(ComponentPose=@n33)) {cacheName=NormalDynamics}
   Save cached pose 'BPPose01'(Pose=Component To Local()) {cacheName=BPPose01}
   n34 = "Constraint - dyn_hair_m_01 (PP)"!(
    ComponentPose=Local To Component(LocalPose=Use cached pose 'StartingPose'())
    Alpha=1
    ConstraintWeights_0=0.9
    ConstraintWeights_1:float=0.1
   ) -> (Pose) {
    alphaBoolBlend {blendOption=Linear}
    boneToModify {boneName=dyn_hair_m_01}
    constraintSetup [{targetBone {boneName=head}}, {targetBone {boneName=spine_05}}]
    constraintWeights [1,1]
   }
   n35 = "Constraint - dyn_hair_m_02 (PP)"!(ComponentPose=@n34, Alpha=1, ConstraintWeights_0=0.8, ConstraintWeights_1:float=0.2) -> (Pose) {
    alphaBoolBlend {blendOption=Linear}
    boneToModify {boneName=dyn_hair_m_02}
    constraintSetup [{targetBone {boneName=head}}, {targetBone {boneName=spine_05}}]
    constraintWeights [1,1]
   }
   n36 = "Constraint - dyn_hair_m_03 (PP)"!(ComponentPose=@n35, Alpha=1, ConstraintWeights_0=0.9, ConstraintWeights_1:float=0.1) -> (Pose) {
    alphaBoolBlend {blendOption=Linear}
    boneToModify {boneName=dyn_hair_m_03}
    constraintSetup [{targetBone {boneName=head}}, {targetBone {boneName=spine_05}}]
    constraintWeights [1,1]
   }
   n37 = "Constraint - dyn_hair_m_04 (PP)"!(ComponentPose=@n36, Alpha=1, ConstraintWeights_0=0.7, ConstraintWeights_1:float=0.3) -> (Pose) {
    alphaBoolBlend {blendOption=Linear}
    boneToModify {boneName=dyn_hair_m_04}
    constraintSetup [{targetBone {boneName=head}}, {targetBone {boneName=spine_05}}]
    constraintWeights [1,1]
   }
   n38 = "Constraint - dyn_hair_m_05 (PP)"!(ComponentPose=@n37, Alpha=1, ConstraintWeights_0=0.7, ConstraintWeights_1:float=0.3) -> (Pose) {
    alphaBoolBlend {blendOption=Linear}
    boneToModify {boneName=dyn_hair_m_05}
    constraintSetup [{targetBone {boneName=head}}, {targetBone {boneName=spine_05}}]
    constraintWeights [1,1]
   }
   n39 = "Constraint - dyn_hair_m_06 (PP)"!(ComponentPose=@n38, Alpha=1, ConstraintWeights_0=0.5, ConstraintWeights_1:float=0.5) -> (Pose) {
    alphaBoolBlend {blendOption=Linear}
    boneToModify {boneName=dyn_hair_m_06}
    constraintSetup [{targetBone {boneName=head}}, {targetBone {boneName=spine_05}}]
    constraintWeights [1,1]
   }
   n40 = "Constraint - dyn_hair_m_07 (PPP)"!(ComponentPose=@n39, Alpha=1, ConstraintWeights_0=0.2, ConstraintWeights_1:float=0.2, ConstraintWeights_2:float=0.6) -> (Pose) {
    alphaBoolBlend {blendOption=Linear}
    boneToModify {boneName=dyn_hair_m_07}
    constraintSetup [{targetBone {boneName=head}}, {targetBone {boneName=spine_05}}, {targetBone {boneName=spine_04}}]
    constraintWeights [1,1,1]
   }
   n41 = "Constraint - dyn_hair_m_08 (PPP)"!(ComponentPose=@n40, Alpha=1, ConstraintWeights_0=0.5, ConstraintWeights_1:float=0.5, ConstraintWeights_2:float=0) -> (Pose) {
    alphaBoolBlend {blendOption=Linear}
    boneToModify {boneName=dyn_hair_m_08}
    constraintSetup [{targetBone {boneName=spine_05}}, {targetBone {boneName=spine_04}}, {targetBone {boneName=head}}]
    constraintWeights [1,1,1]
   }
   n42 = "Constraint - dyn_hair_m_09 (PPP)"!(ComponentPose=@n41, Alpha=1, ConstraintWeights_0=0.5, ConstraintWeights_1:float=0.5, ConstraintWeights_2:float=0) -> (Pose) {
    alphaBoolBlend {blendOption=Linear}
    boneToModify {boneName=dyn_hair_m_09}
    constraintSetup [{targetBone {boneName=spine_05}}, {targetBone {boneName=spine_04}}, {targetBone {boneName=head}}]
    constraintWeights [1,1,1]
   }
   n43 = "Transform (Modify) Bone - Bone: dyn_hair_m_02"!(ComponentPose=@n42, Alpha=1, Translation:Vector="0,0,0", Rotation="10,0,0", Scale:Vector="1,1,1") -> (Pose) {
    rotationMode=BMM_Additive
    rotationSpace=BCS_BoneSpace
    translationMode=BMM_Additive
    translationSpace=BCS_BoneSpace
    alphaBoolBlend {blendOption=Linear}
    alphaScaleBiasClamp {inRange {max=0, min=-25}, outRange {max=0, min=1}}
    boneToModify {boneName=dyn_hair_m_02}
    rotation {pitch=10}
   }
   n44 = "Transform (Modify) Bone - Bone: dyn_hair_m_03"!(ComponentPose=@n43, Alpha=1, Translation:Vector="0,0,0", Rotation="-20,0,0", Scale:Vector="1,1,1") -> (Pose) {
    rotationMode=BMM_Additive
    rotationSpace=BCS_BoneSpace
    translationMode=BMM_Additive
    translationSpace=BCS_BoneSpace
    alphaBoolBlend {blendOption=Linear}
    alphaScaleBiasClamp {inRange {max=0, min=-25}, outRange {max=0, min=1}}
    boneToModify {boneName=dyn_hair_m_03}
    rotation {pitch=-20}
   }
   n45 = "Transform (Modify) Bone - Bone: dyn_hair_m_04"!(ComponentPose=@n44, Alpha=1, Translation:Vector="0,0,0", Rotation="5,0,0", Scale:Vector="1,1,1") -> (Pose) {
    rotationMode=BMM_Additive
    rotationSpace=BCS_BoneSpace
    translationMode=BMM_Additive
    translationSpace=BCS_BoneSpace
    alphaBoolBlend {blendOption=Linear}
    alphaScaleBiasClamp {inRange {max=0, min=-25}, outRange {max=0, min=1}}
    boneToModify {boneName=dyn_hair_m_04}
    rotation {pitch=5}
   }
   n46 = "Transform (Modify) Bone - Bone: dyn_hair_m_05"!(ComponentPose=@n45, Alpha:float=1, Translation="0,0,0", Rotation="10,0,0", Scale:Vector="1,1,1") -> (Pose) {
    rotationMode=BMM_Additive
    rotationSpace=BCS_BoneSpace
    translationMode=BMM_Additive
    translationSpace=BCS_BoneSpace
    alphaBoolBlend {blendOption=Linear}
    alphaScaleBiasClamp {inRange {max=0, min=-25}, outRange {max=0, min=1}}
    boneToModify {boneName=dyn_hair_m_05}
    rotation {pitch=10}
   }
   n47 = "Transform (Modify) Bone - Bone: dyn_hair_m_06"!(ComponentPose=@n46, Alpha=1, Translation:Vector="0,0,0", Rotation="-5,0,0", Scale:Vector="1,1,1") -> (Pose) {
    rotationMode=BMM_Additive
    rotationSpace=BCS_BoneSpace
    translationMode=BMM_Additive
    translationSpace=BCS_BoneSpace
    alphaBoolBlend {blendOption=Linear}
    alphaScaleBiasClamp {inRange {max=0, min=-25}, outRange {max=0, min=1}}
    boneToModify {boneName=dyn_hair_m_06}
    rotation {pitch=-5}
   }
   n48 = "Transform (Modify) Bone - Bone: dyn_hair_m_07"!(ComponentPose=@n47, Alpha=1, Translation:Vector="0,0,0", Rotation="-10,0,0", Scale:Vector="1,1,1") -> (Pose) {
    rotationMode=BMM_Additive
    rotationSpace=BCS_BoneSpace
    translationMode=BMM_Additive
    translationSpace=BCS_BoneSpace
    alphaBoolBlend {blendOption=Linear}
    alphaScaleBiasClamp {inRange {max=0, min=-25}, outRange {max=0, min=1}}
    boneToModify {boneName=dyn_hair_m_07}
    rotation {pitch=-10}
   }
   n49 = "Transform (Modify) Bone - Bone: dyn_hair_m_08"(ComponentPose=@n48, Alpha=1, Translation="0,0,0", Rotation="-29.902,0,0", Scale="1,1,1") {
    rotationMode=BMM_Additive
    rotationSpace=BCS_BoneSpace
    translationMode=BMM_Additive
    translationSpace=BCS_BoneSpace
    alphaBoolBlend {blendOption=Linear}
    alphaScaleBiasClamp {inRange {max=0, min=-25}, outRange {max=0, min=1}}
    boneToModify {boneName=dyn_hair_m_08}
    rotation {pitch=-29.902}
   }
   n50 = "Transform (Modify) Bone - Bone: dyn_hair_m_09"(ComponentPose=@n49, Alpha=1, Translation="0,0,0", Rotation="-82.307,-70.159,71.045", Scale="1,1,1") {
    rotationMode=BMM_Additive
    rotationSpace=BCS_BoneSpace
    translationMode=BMM_Additive
    translationSpace=BCS_BoneSpace
    alphaBoolBlend {blendOption=Linear}
    alphaScaleBiasClamp {inRange {max=0, min=-25}, outRange {max=0, min=1}}
    boneToModify {boneName=dyn_hair_m_09}
    rotation=<rot -82.307 -70.1591 71.0451>
   }
   n51 = "Transform (Modify) Bone - Bone: dyn_hair_m_10"(ComponentPose=@n50, Alpha=1, Translation="0,0,0", Rotation="-53.817,0,0", Scale="1,1,1") {
    rotationMode=BMM_Additive
    rotationSpace=BCS_BoneSpace
    translationMode=BMM_Additive
    translationSpace=BCS_BoneSpace
    alphaBoolBlend {blendOption=Linear}
    alphaScaleBiasClamp {inRange {max=0, min=-25}, outRange {max=0, min=1}}
    boneToModify {boneName=dyn_hair_m_10}
    rotation {pitch=-53.817}
   }
   n52 = Get HeadUpFloat()
   n53 = "Transform (Modify) Bone - Bone: dyn_hair_m_02"(ComponentPose=@n51, Alpha=@n52, Translation="0,0,0", Rotation="0,0,0", Scale="1,1,1") {
    rotationMode=BMM_Additive
    rotationSpace=BCS_BoneSpace
    translationMode=BMM_Additive
    translationSpace=BCS_BoneSpace
    alphaBoolBlend {blendOption=Linear}
    alphaScaleBiasClamp {inRange {max=0, min=-25}, outRange {max=0, min=1}}
    boneToModify {boneName=dyn_hair_m_02}
   }
   n54 = "Transform (Modify) Bone - Bone: dyn_hair_m_03"(ComponentPose=@n53, Alpha=@n52, Translation="0,0,0", Rotation="0,0,0", Scale="1,1,1") {
    rotationMode=BMM_Additive
    rotationSpace=BCS_BoneSpace
    translationMode=BMM_Additive
    translationSpace=BCS_BoneSpace
    alphaBoolBlend {blendOption=Linear}
    alphaScaleBiasClamp {inRange {max=0, min=-25}, outRange {max=0, min=1}}
    boneToModify {boneName=dyn_hair_m_03}
   }
   n55 = "Transform (Modify) Bone - Bone: dyn_hair_m_04"(ComponentPose=@n54, Alpha=@n52, Translation="0,0,0", Rotation="0,0,0", Scale="1,1,1") {
    rotationMode=BMM_Additive
    rotationSpace=BCS_BoneSpace
    translationMode=BMM_Additive
    translationSpace=BCS_BoneSpace
    alphaBoolBlend {blendOption=Linear}
    alphaScaleBiasClamp {inRange {max=0, min=-25}, outRange {max=0, min=1}}
    boneToModify {boneName=dyn_hair_m_04}
   }
   n56 = "Transform (Modify) Bone - Bone: dyn_hair_m_05"(ComponentPose=@n55, Alpha=@n52, Translation="0,0,0", Rotation="0,0,0", Scale="1,1,1") {
    rotationMode=BMM_Additive
    rotationSpace=BCS_BoneSpace
    translationMode=BMM_Additive
    translationSpace=BCS_BoneSpace
    alphaBoolBlend {blendOption=Linear}
    alphaScaleBiasClamp {inRange {max=0, min=-25}, outRange {max=0, min=1}}
    boneToModify {boneName=dyn_hair_m_05}
   }
   n57 = "Transform (Modify) Bone - Bone: dyn_hair_m_06"(ComponentPose=@n56, Alpha=@n52, Translation="0,0,0", Rotation="39.92,0,0", Scale="1,1,1") {
    rotationMode=BMM_Additive
    rotationSpace=BCS_BoneSpace
    translationMode=BMM_Additive
    translationSpace=BCS_BoneSpace
    alphaBoolBlend {blendOption=Linear}
    alphaScaleBiasClamp {inRange {max=0, min=-25}, outRange {max=0, min=1}}
    boneToModify {boneName=dyn_hair_m_06}
    rotation {pitch=39.92}
   }
   n58 = "Transform (Modify) Bone - Bone: dyn_hair_m_07"(ComponentPose=@n57, Alpha=@n52, Translation="4.810,-0.043,-0.130", Rotation="-10.017,0,0", Scale="1,1,1") {
    rotationMode=BMM_Additive
    rotationSpace=BCS_BoneSpace
    translationMode=BMM_Additive
    translationSpace=BCS_BoneSpace
    alphaBoolBlend {blendOption=Linear}
    alphaScaleBiasClamp {inRange {max=0, min=-25}, outRange {max=0, min=1}}
    boneToModify {boneName=dyn_hair_m_07}
    rotation {pitch=-10.017}
    translation=<4.81 -0.043 -0.13>
   }
   n59 = "Transform (Modify) Bone - Bone: dyn_hair_m_08"(ComponentPose=@n58, Alpha=@n52, Translation="0,0,-5", Rotation="-19.912,0,0", Scale="1,1,1") {
    rotationMode=BMM_Additive
    rotationSpace=BCS_BoneSpace
    translationMode=BMM_Additive
    translationSpace=BCS_BoneSpace
    alphaBoolBlend {blendOption=Linear}
    alphaScaleBiasClamp {inRange {max=0, min=-25}, outRange {max=0, min=1}}
    boneToModify {boneName=dyn_hair_m_08}
    rotation {pitch=-19.912}
    translation {z=-5}
   }
   n60 = "Transform (Modify) Bone - Bone: dyn_hair_m_09"(ComponentPose=@n59, Alpha=@n52, Translation="0,0,0", Rotation="19.173,0,2.3804", Scale="1,1,1") {
    rotationMode=BMM_Additive
    rotationSpace=BCS_BoneSpace
    translationMode=BMM_Additive
    translationSpace=BCS_BoneSpace
    alphaBoolBlend {blendOption=Linear}
    alphaScaleBiasClamp {inRange {max=0, min=-25}, outRange {max=0, min=1}}
    boneToModify {boneName=dyn_hair_m_09}
    rotation {pitch=19.173, roll=2.3804}
   }
   n61 = Get Head Down()
   n62 = "Transform (Modify) Bone - Bone: dyn_hair_m_02"(ComponentPose=@n60, Alpha=@n61, Translation="0,0,0", Rotation="10,0,0", Scale="1,1,1") {
    rotationMode=BMM_Additive
    rotationSpace=BCS_BoneSpace
    translationMode=BMM_Additive
    translationSpace=BCS_BoneSpace
    alphaBoolBlend {blendOption=Linear}
    alphaScaleBiasClamp {inRange {max=0, min=-25}, outRange {max=0, min=1}}
    boneToModify {boneName=dyn_hair_m_02}
    rotation {pitch=10}
   }
   n63 = "Transform (Modify) Bone - Bone: dyn_hair_m_03"(ComponentPose=@n62, Alpha=@n61, Translation="0,0,0", Rotation="-20,0,0", Scale="1,1,1") {
    rotationMode=BMM_Additive
    rotationSpace=BCS_BoneSpace
    translationMode=BMM_Additive
    translationSpace=BCS_BoneSpace
    alphaBoolBlend {blendOption=Linear}
    alphaScaleBiasClamp {inRange {max=0, min=-25}, outRange {max=0, min=1}}
    boneToModify {boneName=dyn_hair_m_03}
    rotation {pitch=-20}
   }
   n64 = "Transform (Modify) Bone - Bone: dyn_hair_m_04"(ComponentPose=@n63, Alpha=@n61, Translation="0,0,0", Rotation="15,0,0", Scale="1,1,1") {
    rotationMode=BMM_Additive
    rotationSpace=BCS_BoneSpace
    translationMode=BMM_Additive
    translationSpace=BCS_BoneSpace
    alphaBoolBlend {blendOption=Linear}
    alphaScaleBiasClamp {inRange {max=0, min=-25}, outRange {max=0, min=1}}
    boneToModify {boneName=dyn_hair_m_04}
    rotation {pitch=15}
   }
   n65 = "Transform (Modify) Bone - Bone: dyn_hair_m_05"(ComponentPose=@n64, Alpha=@n61, Translation="0,0,0", Rotation="0,0,0", Scale="1,1,1") {
    rotationMode=BMM_Additive
    rotationSpace=BCS_BoneSpace
    translationMode=BMM_Additive
    translationSpace=BCS_BoneSpace
    alphaBoolBlend {blendOption=Linear}
    alphaScaleBiasClamp {inRange {max=0, min=-25}, outRange {max=0, min=1}}
    boneToModify {boneName=dyn_hair_m_05}
   }
   n66 = "Transform (Modify) Bone - Bone: dyn_hair_m_06"(ComponentPose=@n65, Alpha=@n61, Translation="0,0,0", Rotation="0,0,0", Scale="1,1,1") {
    rotationMode=BMM_Additive
    rotationSpace=BCS_BoneSpace
    translationMode=BMM_Additive
    translationSpace=BCS_BoneSpace
    alphaBoolBlend {blendOption=Linear}
    alphaScaleBiasClamp {inRange {max=0, min=-25}, outRange {max=0, min=1}}
    boneToModify {boneName=dyn_hair_m_06}
   }
   n67 = "Transform (Modify) Bone - Bone: dyn_hair_m_07"(ComponentPose=@n66, Alpha=@n61, Translation="0,0,0", Rotation="0,0,0", Scale="1,1,1") {
    rotationMode=BMM_Additive
    rotationSpace=BCS_BoneSpace
    translationMode=BMM_Additive
    translationSpace=BCS_BoneSpace
    alphaBoolBlend {blendOption=Linear}
    alphaScaleBiasClamp {inRange {max=0, min=-25}, outRange {max=0, min=1}}
    boneToModify {boneName=dyn_hair_m_07}
   }
   n68 = "Transform (Modify) Bone - Bone: dyn_hair_m_08"(ComponentPose=@n67, Alpha=@n61, Translation="0,0,0", Rotation="0,0,0", Scale="1,1,1") {
    rotationMode=BMM_Additive
    rotationSpace=BCS_BoneSpace
    translationMode=BMM_Additive
    translationSpace=BCS_BoneSpace
    alphaBoolBlend {blendOption=Linear}
    alphaScaleBiasClamp {inRange {max=0, min=-25}, outRange {max=0, min=1}}
    boneToModify {boneName=dyn_hair_m_08}
   }
   n69 = "Transform (Modify) Bone - Bone: dyn_hair_m_09"(ComponentPose=@n68, Alpha=@n61, Translation="0,0,0", Rotation="0,0,0", Scale="1,1,1") {
    rotationMode=BMM_Additive
    rotationSpace=BCS_BoneSpace
    translationMode=BMM_Additive
    translationSpace=BCS_BoneSpace
    alphaBoolBlend {blendOption=Linear}
    alphaScaleBiasClamp {inRange {max=0, min=-25}, outRange {max=0, min=1}}
    boneToModify {boneName=dyn_hair_m_09}
   }
   n70 = Get HeadUpCrouchWalk() -> (HeadUpCrouchWalk:double)
   n71 = "Transform (Modify) Bone - Bone: dyn_hair_m_02"(ComponentPose=@n69, Alpha=@n70, Translation="0,0,0", Rotation="0,0,0", Scale="1,1,1") {
    rotationMode=BMM_Additive
    rotationSpace=BCS_BoneSpace
    translationMode=BMM_Additive
    translationSpace=BCS_BoneSpace
    alphaBoolBlend {blendOption=Linear}
    alphaScaleBiasClamp {inRange {max=0, min=-25}, outRange {max=0, min=1}}
    boneToModify {boneName=dyn_hair_m_02}
   }
   n72 = "Transform (Modify) Bone - Bone: dyn_hair_m_03"(ComponentPose=@n71, Alpha=@n70, Translation="0,0,0", Rotation="0,0,0", Scale="1,1,1") {
    rotationMode=BMM_Additive
    rotationSpace=BCS_BoneSpace
    translationMode=BMM_Additive
    translationSpace=BCS_BoneSpace
    alphaBoolBlend {blendOption=Linear}
    alphaScaleBiasClamp {inRange {max=0, min=-25}, outRange {max=0, min=1}}
    boneToModify {boneName=dyn_hair_m_03}
   }
   n73 = "Transform (Modify) Bone - Bone: dyn_hair_m_04"(ComponentPose=@n72, Alpha=@n70, Translation="5.143,-0.069,1.247", Rotation="19.968,1.096,-0.255", Scale="1,1,1") {
    rotationMode=BMM_Additive
    rotationSpace=BCS_BoneSpace
    translationMode=BMM_Additive
    translationSpace=BCS_BoneSpace
    alphaBoolBlend {blendOption=Linear}
    alphaScaleBiasClamp {inRange {max=0, min=-25}, outRange {max=0, min=1}}
    boneToModify {boneName=dyn_hair_m_04}
    rotation=<rot 19.9679 1.096 -0.255>
    translation=<5.143 -0.069 1.247>
   }
   n74 = "Transform (Modify) Bone - Bone: dyn_hair_m_05"(ComponentPose=@n73, Alpha=@n70, Translation="0,0,0", Rotation="9.979,0.702,0.199", Scale="1,1,1") {
    rotationMode=BMM_Additive
    rotationSpace=BCS_BoneSpace
    translationMode=BMM_Additive
    translationSpace=BCS_BoneSpace
    alphaBoolBlend {blendOption=Linear}
    alphaScaleBiasClamp {inRange {max=0, min=-25}, outRange {max=0, min=1}}
    boneToModify {boneName=dyn_hair_m_05}
    rotation=<rot 9.9789 0.702 0.199>
   }
   n75 = "Transform (Modify) Bone - Bone: dyn_hair_m_06"(ComponentPose=@n74, Alpha=@n70, Translation="3.437,0.120,-0.426", Rotation="-19.938,-1.223,1.298", Scale="1,1,1") {
    rotationMode=BMM_Additive
    rotationSpace=BCS_BoneSpace
    translationMode=BMM_Additive
    translationSpace=BCS_BoneSpace
    alphaBoolBlend {blendOption=Linear}
    alphaScaleBiasClamp {inRange {max=0, min=-25}, outRange {max=0, min=1}}
    boneToModify {boneName=dyn_hair_m_06}
    rotation=<rot -19.938 -1.223 1.298>
    translation=<3.437 0.12 -0.426>
   }
   n76 = "Transform (Modify) Bone - Bone: dyn_hair_m_07"(ComponentPose=@n75, Alpha=@n70, Translation="2.438,0.302,0.646", Rotation="-9.964,1.091,1.977", Scale="1,1,1") {
    rotationMode=BMM_Additive
    rotationSpace=BCS_BoneSpace
    translationMode=BMM_Additive
    translationSpace=BCS_BoneSpace
    alphaBoolBlend {blendOption=Linear}
    alphaScaleBiasClamp {inRange {max=0, min=-25}, outRange {max=0, min=1}}
    boneToModify {boneName=dyn_hair_m_07}
    rotation=<rot -9.964 1.091 1.977>
    translation=<2.438 0.302 0.646>
   }
   n77 = "Transform (Modify) Bone - Bone: dyn_hair_m_08"(ComponentPose=@n76, Alpha=@n70, Translation="0,0,0", Rotation="-19.543,-2.027,4.394", Scale="1,1,1") {
    rotationMode=BMM_Additive
    rotationSpace=BCS_BoneSpace
    translationMode=BMM_Additive
    translationSpace=BCS_BoneSpace
    alphaBoolBlend {blendOption=Linear}
    alphaScaleBiasClamp {inRange {max=0, min=-25}, outRange {max=0, min=1}}
    boneToModify {boneName=dyn_hair_m_08}
    rotation=<rot -19.5431 -2.027 4.394>
   }
   n78 = "Transform (Modify) Bone - Bone: dyn_hair_m_09"(ComponentPose=@n77, Alpha=@n70, Translation="0,0,0", Rotation="-9.081,-3.904,2.069", Scale="1,1,1") {
    rotationMode=BMM_Additive
    rotationSpace=BCS_BoneSpace
    translationMode=BMM_Additive
    translationSpace=BCS_BoneSpace
    alphaBoolBlend {blendOption=Linear}
    alphaScaleBiasClamp {inRange {max=0, min=-25}, outRange {max=0, min=1}}
    boneToModify {boneName=dyn_hair_m_09}
    rotation=<rot -9.0811 -3.904 2.069>
   }
   n79 = RigidBody(ComponentPose=@n78, Alpha=0.3, OverrideWorldGravity=Get GravityOverride()) {
    bOverrideWorldGravity=true
    overridePhysicsAsset=&/M_MED_MagmaBreak_FaceAcc_BackPack_Physics
    simulationSpace=BaseBoneSpace
    alphaBoolBlend {blendOption=Linear}
    baseBoneRef {boneName=head}
    componentAppliedLinearAccClamp=<6000 4000 2500>
    componentLinearAccScale=<0.4 0.25 0.25>
    componentLinearVelScale=<0.6 0.4 0.25>
    simSpaceSettings {
     maxAngularAcceleration=1000
     maxAngularVelocity=3
     maxLinearAcceleration=700
     maxLinearVelocity=500
     velocityScaleZ=0
     worldAlpha=0.2
     externalLinearDragV=<0.2 0.2>
    }
   }
   Save cached pose 'Backpack Dynamics'(Pose=Component To Local(ComponentPose=@n79)) {cacheName=Backpack Dynamics}
   n80 = RigidBody(ComponentPose=Local To Component(LocalPose=Use cached pose 'StartingPose'()), OverrideWorldGravity=Get GravityOverride()) {
    bClampLinearTranslationLimitToRefPose=true
    bOverrideWorldGravity=true
    overridePhysicsAsset=&M_MED_MagmaBreak_FaceAcc_Physics
    simulationSpace=BaseBoneSpace
    alphaBoolBlend {blendOption=Linear}
    baseBoneRef {boneName=head}
    componentLinearAccScale=<0.3 0.24 0.24>
    componentLinearVelScale=<0.3 0.3 0.3>
    overrideWorldGravity {z=-980}
    simSpaceSettings {maxAngularAcceleration=1000, maxAngularVelocity=3, maxLinearAcceleration=250, maxLinearVelocity=500, velocityScaleZ=0, worldAlpha=0.25}
   }
   n81 = "Transform (Modify) Bone - Bone: dyn_hair_m_02"(ComponentPose=@n80, Alpha=1, Translation="0,0,0", Rotation="10,0.021,0", Scale="1,1,1") {
    rotationMode=BMM_Additive
    rotationSpace=BCS_BoneSpace
    translationMode=BMM_Additive
    translationSpace=BCS_BoneSpace
    alphaBoolBlend {blendOption=Linear}
    boneToModify {boneName=dyn_hair_m_02}
    rotation {pitch=10, yaw=0.021}
   }
   n82 = "Transform (Modify) Bone - Bone: dyn_hair_m_03"(ComponentPose=@n81, Alpha=1, Translation="0,0,0", Rotation="5,0,0", Scale="1,1,1") {
    rotationMode=BMM_Additive
    rotationSpace=BCS_BoneSpace
    translationMode=BMM_Additive
    translationSpace=BCS_BoneSpace
    alphaBoolBlend {blendOption=Linear}
    boneToModify {boneName=dyn_hair_m_03}
    rotation {pitch=5}
   }
   n83 = "Transform (Modify) Bone - Bone: dyn_hair_m_04"(ComponentPose=@n82, Alpha=1, Translation="0,0,0", Rotation="0,0,0", Scale="1,1,1") {
    rotationMode=BMM_Additive
    rotationSpace=BCS_BoneSpace
    translationMode=BMM_Additive
    translationSpace=BCS_BoneSpace
    alphaBoolBlend {blendOption=Linear}
    boneToModify {boneName=dyn_hair_m_04}
   }
   n84 = "Transform (Modify) Bone - Bone: dyn_hair_m_05"(ComponentPose=@n83, Alpha=1, Translation="0,0,0", Rotation="0,0,0", Scale="1,1,1") {
    rotationMode=BMM_Additive
    rotationSpace=BCS_BoneSpace
    translationMode=BMM_Additive
    translationSpace=BCS_BoneSpace
    alphaBoolBlend {blendOption=Linear}
    boneToModify {boneName=dyn_hair_m_05}
   }
   n85 = "Transform (Modify) Bone - Bone: dyn_hair_m_06"(ComponentPose=@n84, Alpha=1, Translation="0,0,0", Rotation="0,0,0", Scale="1,1,1") {
    rotationMode=BMM_Additive
    rotationSpace=BCS_BoneSpace
    translationMode=BMM_Additive
    translationSpace=BCS_BoneSpace
    alphaBoolBlend {blendOption=Linear}
    alphaScaleBiasClamp {bInterpResult=true}
    boneToModify {boneName=dyn_hair_m_06}
   }
   Save cached pose 'Cape Dynamics'(Pose=Component To Local(ComponentPose=@n85)) {cacheName=Cape Dynamics}
   Save cached pose 'StartingPose'(Pose=Copy Pose From Mesh(SourceMeshComponent=Get Skel Mesh()) {bUseAttachedParent=true}) {cacheName=StartingPose}
   Use cached pose 'NormalDynamics'()
   "To Float (Boolean)"(InBool=@n29)
   Get ISCrouchwalkingFloat() -> (ISCrouchwalkingFloat:double)
   Get head_fwdBwd_rot() -> (head_fwdBwd_rot:double)
   Get Head Down()
   Get HeadUpFloat()
   Get HeadUpCrouchWalk()
   Get HeadUpFloat()
  ]}
 EventGraph {nodes [
   Event BlueprintUpdateAnimation()
   > n1 = Sequence!(execute) -> (then_0, then_1, then_2:Exec, then_3:Exec, then_4:Exec, then_5:Exec)
   > n2 = Gravity Override State Function DEPRECATED(CharacterPartAnimInstance=Self-Reference(), GameTimeInSeconds=GetGameTimeInSeconds(), GravityOverrideParams=Get Gravity Override Params())
   > Set GravityOverride(GravityOverride=@n2)
   Set IsCrouchWalking(execute=@n1.then_5, IsCrouchWalking:bool=AND Boolean(A=Get bIsCrouching(), B=Get bIsMoving2D())) -> (Output_Get:bool)
   n3 = Get is Wearing Backbling()
   n4 = Branch(execute=@n1.then_4, Condition="Nearly Equal (Float)"(A="To Float (Boolean)"(InBool=@n3), B=1, ErrorTolerance=0.1))
   n5 = GetDeltaTransformFromRefPose(self=GetOwnerSkeletalMeshComponent(), BoneName=head, BaseName=neck_01)
   > Set head_fwdBwd_rot(head_fwdBwd_rot:double=BreakRotator(InRot=BreakTransform(InTransform=@n5).Rotation).Yaw) -> (Output_Get:double)
   Set head_fwdBwd_rot(execute=@n4.else, head_fwdBwd_rot=0)
   n6 = Get Skel Mesh()
   Set Head Down(
    execute=@n1.then_2
    Head Down:double=float * float(
     A=GetAlphaFromReferenceTransformDelta(SkeletalMeshComponent=@n6, BoneName=spine_05, BaseName="head ", InputRange=(X=0,Y=-70), OutputRange=(X=0,Y=-1), TwistAxis="0,0,1")
     B=-10
    ) {is=title K2Node_PromotableOperator}
   ) -> (Output_Get:double)
   > Set Head_Turn_L(Head_Turn_L:double=GetAlphaFromReferenceTransformDelta(SkeletalMeshComponent=@n6, BoneName=spine_05, BaseName=head, DeltaComponentType=EulerX, InputRange=(X=0,Y=-22), TwistAxis="1,0,0")) -> (Output_Get:double)
   > Set Head_Turn_R(Head_Turn_R:double=GetAlphaFromReferenceTransformDelta(SkeletalMeshComponent=@n6, BoneName=spine_05, BaseName=head, DeltaComponentType=EulerX, InputRange=(X=0,Y=22), TwistAxis="1,0,0")) -> (Output_Get:double)
   n7 = Set HeadUpFloat(
    execute=@n1.then_1
    HeadUpFloat:double=GetAlphaFromReferenceTransformDelta(SkeletalMeshComponent=@n6, BoneName=spine_05, BaseName=head, InputRange=(X=10,Y=70), TwistAxis="0,0,1")
   ) -> (Output_Get:double)
   n8 = Get bIsCrouching()
   > Set HeadUpCrouchTargeting(HeadUpCrouchTargeting:double=float * float!(A=@n7, B:bool=@n8, C:bool=Get bIsTargeting()) -> (ReturnValue)) -> (Output_Get:double)
   > Set HeadUpCrouchWalk(HeadUpCrouchWalk:double=float * float!(A=Get HeadUpFloat(), B=Get IsCrouchWalking()) -> (ReturnValue)) -> (Output_Get:double)
   > Set ISCrouchwalkingFloat(ISCrouchwalkingFloat:double="To Float (Boolean)"(InBool=Get IsCrouchWalking())) -> (Output_Get:double)
   n9 = CheckForBackbling()
   Set is Wearing Backbling(
    execute=@n1.then_3
    is Wearing Backbling:bool=AND Boolean!(A=@n9, B=NOT Boolean(A=@n9.bIsBackblingACape), C:bool=NOT Boolean(A=@n9.bIsBackblingLong)) -> (ReturnValue)
   ) -> (Output_Get:bool)
   > Set is Wearing Cape(is Wearing Cape:bool=OR Boolean(A=@n9.bIsBackblingACape, B=@n9.bIsBackblingLong)) -> (Output_Get:bool)
   Event BlueprintInitializeAnimation()
   > Set Skel Mesh(Skel Mesh=GetOwnerSkeletalMeshComponent())
   "To Float (Boolean)"(InBool=@n8)
  ]}
]
variables [
 Gravity Override Params {type=GravityOverrideParamsStruct}
 GravityOverride {type=Vector}
 Head Down {type=number}
 head_fwdBwd_rot {type=number}
 Head_Turn_L {type=number}
 Head_Turn_R {type=number}
 HeadUpCrouchTargeting {type=number}
 HeadUpCrouchWalk {type=number}
 HeadUpFloat {type=number}
 is Wearing Backbling {type=boolean}
 is Wearing Cape {type=boolean}
 IsCrouchWalking {type=boolean}
 ISCrouchwalkingFloat {type=number}
 Skel Mesh {type=SkeletalMeshComponent}
]
